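# Physics asset for the FruitFire Nudge pickaxe, setting up the dangling cord on the two-handed handle for simulation.
# Seven small capsule bodies chain from the handle through dyn_cord_MSTR to a larger end capsule, with limited swing and very tight twist.
# FruitFire pelvis, thigh and calf capsules, apparently body colliders for the cord, each constrain directly to the handle with free angular motion.
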PhysicsAsset /BRCosmetics/Weapons/FORT_Melee/Pickaxe_FruitFire_Nudge/Meshes/FruitFire_Nudge_Axe_Cord_Physics
bodies {
 FruitFire_calf_l [{shapeType=Capsule, length=50, radius=8.5, center=<-30 0 0>, rotation=<rot 90 0 0>}]
 FruitFire_calf_r [{shapeType=Capsule, length=50, radius=8.5, center=<30 0 0>, rotation=<rot 90 0 0>}]
 FruitFire_pelvis [{shapeType=Capsule, length=10, radius=15, center=<-2.4887 0.1881 0>}]
 FruitFire_thigh_l [{shapeType=Capsule, length=40, radius=8.5, center=<-25 0 0>, rotation=<rot 90 0 0>}]
 FruitFire_thigh_r [{shapeType=Capsule, length=40, radius=8.5, center=<25 0 0>, rotation=<rot 90 0 0>}]
 Melee_TwoHanded_Handle [
  {shapeType=Capsule, length=19.5166, radius=2.2365, center=<-0.1223 -0.3623 -6.7577>, rotation=<rot 0.4874 -116.0453 0.4889>}
  {shapeType=Capsule, length=11.7, radius=3, shapeName=Melee_TwoHanded_Handle_capsule, center=<0 -2.8249 -24.9695>, rotation=<rot 0 0 15.7299>}
 ]
 dyn_cord_01 [{shapeType=Capsule, length=1.5, radius=1, shapeName=dyn_cord_01_capsule, center=<3.3065 -0.0038 0.0506>, rotation=<rot 0 -87.9358 -80.4129>}]
 dyn_cord_02 [{shapeType=Capsule, length=1.5, radius=1, shapeName=dyn_cord_02_capsule, center=<3.8379 -0.0264 -0.1657>, rotation=<rot 0 91.0845 -87.9533>}]
 dyn_cord_03 [{shapeType=Capsule, length=1.5, radius=1, shapeName=dyn_cord_03_capsule, center=<4.3386 0.0206 -0.0363>, rotation=<rot 0 89.5679 -88.8529>}]
 dyn_cord_04 [{shapeType=Capsule, length=1.5, radius=1, shapeName=dyn_cord_04_capsule, center=<4.2005 0.0102 0.0104>, rotation=<rot 0 -90.4273 -89.605>}]
 dyn_cord_05 [{shapeType=Capsule, length=1.5, radius=1, shapeName=dyn_cord_05_capsule, center=<4.5211 0 0>, rotation=<rot 0 -89.9453 -89.996>}]
 dyn_cord_06 [{
   shapeType=Capsule
   length=4.4043
   radius=2.5439
   shapeName=dyn_cord_06_capsule
   center=<7.35 0.0976 0.0104>
   rotation=<rot 0 -89.2856 -89.7615>
  }]
 dyn_cord_MSTR [{shapeType=Capsule, length=0.505, radius=0.505, shapeName=dyn_cord_MSTR_capsule, rotation=<rot 0 -90 0>}]
}
constraints [
 {bone1Name=dyn_cord_MSTR, bone2Name=Melee_TwoHanded_Handle, swing1LimitDegrees=30, swing2LimitDegrees=30, twistLimitDegrees=2}
 {bone1Name=dyn_cord_01, bone2Name=dyn_cord_MSTR, swing1LimitDegrees=30, swing2LimitDegrees=30, twistLimitDegrees=2}
 {bone1Name=dyn_cord_02, bone2Name=dyn_cord_01, swing1LimitDegrees=30, swing2LimitDegrees=30, twistLimitDegrees=2}
 {bone1Name=dyn_cord_03, bone2Name=dyn_cord_02, swing1LimitDegrees=30, swing2LimitDegrees=30, twistLimitDegrees=2}
 {bone1Name=dyn_cord_04, bone2Name=dyn_cord_03, swing1LimitDegrees=30, swing2LimitDegrees=30, twistLimitDegrees=2}
 {bone1Name=dyn_cord_05, bone2Name=dyn_cord_04, swing1LimitDegrees=30, swing2LimitDegrees=30, twistLimitDegrees=2}
 {bone1Name=dyn_cord_06, bone2Name=dyn_cord_05, swing1LimitDegrees=30, swing2LimitDegrees=30, twistLimitDegrees=2}
 {
  bone1Name=FruitFire_calf_l
  bone2Name=Melee_TwoHanded_Handle
  swing1LimitDegrees=45
  swing1Motion=Free
  swing2LimitDegrees=45
  swing2Motion=Free
  twistLimitDegrees=45
  twistMotion=Free
 }
 {
  bone1Name=FruitFire_calf_r
  bone2Name=Melee_TwoHanded_Handle
  swing1LimitDegrees=45
  swing1Motion=Free
  swing2LimitDegrees=45
  swing2Motion=Free
  twistLimitDegrees=45
  twistMotion=Free
 }
 {
  bone1Name=FruitFire_pelvis
  bone2Name=Melee_TwoHanded_Handle
  swing1LimitDegrees=45
  swing1Motion=Free
  swing2LimitDegrees=45
  swing2Motion=Free
  twistLimitDegrees=45
  twistMotion=Free
 }
 {
  bone1Name=FruitFire_thigh_l
  bone2Name=Melee_TwoHanded_Handle
  swing1LimitDegrees=45
  swing1Motion=Free
  swing2LimitDegrees=45
  swing2Motion=Free
  twistLimitDegrees=45
  twistMotion=Free
 }
 {
  bone1Name=FruitFire_thigh_r
  bone2Name=Melee_TwoHanded_Handle
  swing1LimitDegrees=45
  swing1Motion=Free
  swing2LimitDegrees=45
  swing2Motion=Free
  twistLimitDegrees=45
  twistMotion=Free
 }
]
