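# Animation Blueprint for the FruitFire Nudge pickaxe that adds secondary motion to its cord, tie and bag bones.
# Cords use either AnimDynamics or, outside the front end when enabled, RigidBody simulation, then feed tie and bag AnimDynamics.
# bIsOnPawn switches between differently tuned cord and tie chains.
# The event graph caches the parent body's pelvis and leg world transforms, matching world-space Modify Bone proxies posed before the RigidBody step.
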AnimBlueprint /BRCosmetics/Weapons/FORT_Melee/Pickaxe_FruitFire_Nudge/Meshes/FruitFire_Nudge_Axe_AnimBP
target_skeleton=&/FruitFire_Nudge_Axe_Skeleton
graphs [
 AnimGraph {nodes [
   n1 = Anim Dynamics - Bone: dyn_tie_01!(ComponentPose, bAlphaBoolEnabled:bool=True, ExternalForce="0,0,0") -> (Pose) {
    alphaInputType=Bool
    angularDampingOverride=0.9998
    bOverrideAngularDamping=true
    bUseGravityOverride=true
    simulationSpace=Actor
    boundBone {boneName=dyn_tie_01}
    componentAppliedLinearAccClamp=<1000 1000 1000>
    componentLinearAccScale=<0.7 0.7 1>
    componentLinearVelScale=<0.8 0.8 1>
    gravityOverride {z=-980}
    physicsBodyDefinitions [{
      collisionType=CustomSphere
      sphereCollisionRadius=1.5
      boundBone {boneName=dyn_tie_01}
      boxExtents=<13 13 13>
      constraintSetup {angularLimitsMax=<0 30 0>, angularLimitsMin=<0 -40 -40>, angularTarget=<0 0 0>}
      localJointOffset=<1.5 0 0>
     }]
    planarLimits [{drivingBone {boneName=dyn_tie_MSTR}, planeTransform {location=<-40 1.5 0>, rotation=<rot 90 90 0>}}]
    simSpaceSettings {maxAngularAcceleration=10, maxAngularVelocity=10, simSpaceAngularAlpha=0.1}
   }
   n2 = Anim Dynamics - Bone: dyn_tie_02!(ComponentPose=@n1, bAlphaBoolEnabled:bool=True, ExternalForce="0,0,0") -> (Pose) {
    alphaInputType=Bool
    angularDampingOverride=0.9998
    bOverrideAngularDamping=true
    bUseGravityOverride=true
    simulationSpace=Actor
    boundBone {boneName=dyn_tie_02}
    componentAppliedLinearAccClamp=<1000 1000 1000>
    componentLinearAccScale=<0.7 0.7 1>
    componentLinearVelScale=<0.8 0.8 1>
    gravityOverride {z=-980}
    physicsBodyDefinitions [{
      collisionType=CustomSphere
      sphereCollisionRadius=1.5
      boundBone {boneName=dyn_tie_02}
      boxExtents=<12 12 12>
      constraintSetup {angularLimitsMax=<5 30 30>, angularLimitsMin=<-5 -40 -40>, angularTarget=<0 0 0>}
      localJointOffset=<1.5 0 0>
     }]
    planarLimits [{drivingBone {boneName=dyn_tie_MSTR}, planeTransform {location=<-40 1.5 0>, rotation=<rot 90 90 0>}}]
    simSpaceSettings {maxAngularAcceleration=10, maxAngularVelocity=10, simSpaceAngularAlpha=0.1}
   }
   n3 = Anim Dynamics - Bone: dyn_tie_03!(ComponentPose=@n2, bAlphaBoolEnabled:bool=True, ExternalForce="0,0,0") -> (Pose) {
    alphaInputType=Bool
    angularDampingOverride=0.999
    bOverrideAngularDamping=true
    bUseGravityOverride=true
    simulationSpace=Actor
    boundBone {boneName=dyn_tie_03}
    componentAppliedLinearAccClamp=<1000 1000 1000>
    componentLinearAccScale=<0.7 0.7 1>
    componentLinearVelScale=<0.8 0.8 1>
    gravityOverride {z=-980}
    physicsBodyDefinitions [{
      collisionType=CustomSphere
      sphereCollisionRadius=1.5
      boundBone {boneName=dyn_tie_03}
      boxExtents=<11 11 11>
      constraintSetup {angularLimitsMax=<5 30 30>, angularLimitsMin=<-5 -30 -40>, angularTarget=<0 0 0>}
      localJointOffset=<1.5 0 0>
     }]
    planarLimits [{drivingBone {boneName=dyn_tie_MSTR}, planeTransform {location=<-40 1.5 0>, rotation=<rot 90 90 0>}}]
    simSpaceSettings {maxAngularAcceleration=10, maxAngularVelocity=10, simSpaceAngularAlpha=0.1}
   }
   Anim Dynamics - Bone: dyn_tie_04!(ComponentPose=@n3, bAlphaBoolEnabled:bool=True, ExternalForce="0,0,0") -> (Pose) {
    alphaInputType=Bool
    angularDampingOverride=0.999
    bOverrideAngularDamping=true
    bUseGravityOverride=true
    simulationSpace=Actor
    boundBone {boneName=dyn_tie_04}
    componentAppliedLinearAccClamp=<1000 1000 1000>
    componentLinearAccScale=<0.7 0.7 1>
    componentLinearVelScale=<0.8 0.8 1>
    gravityOverride {z=-980}
    physicsBodyDefinitions [{
      collisionType=CustomSphere
      sphereCollisionRadius=1.5
      boundBone {boneName=dyn_tie_04}
      boxExtents=<10 10 10>
      constraintSetup {angularLimitsMax=<5 30 30>, angularLimitsMin=<-5 -30 -40>, angularTarget=<0 0 0>}
      localJointOffset=<1.5 0 0>
     }]
    planarLimits [{drivingBone {boneName=dyn_tie_MSTR}, planeTransform {location=<-40 1.5 0>, rotation=<rot 90 90 0>}}]
    simSpaceSettings {maxAngularAcceleration=10, maxAngularVelocity=10, simSpaceAngularAlpha=0.1}
   }
   n4 = Anim Dynamics - Bone: dyn_tie_01(bAlphaBoolEnabled=True, ExternalForce="0,0,0") {
    alphaInputType=Bool
    angularDampingOverride=0.9998
    bOverrideAngularDamping=true
    bUseGravityOverride=true
    simulationSpace=Actor
    boundBone {boneName=dyn_tie_01}
    componentAppliedLinearAccClamp=<1000 1000 1000>
    componentLinearAccScale=<0.7 0.7 1>
    componentLinearVelScale=<0.8 0.8 1>
    gravityOverride {z=-980}
    physicsBodyDefinitions [{
      collisionType=CustomSphere
      sphereCollisionRadius=1.5
      boundBone {boneName=dyn_tie_01}
      boxExtents=<13 13 13>
      constraintSetup {angularLimitsMax=<0 30 0>, angularLimitsMin=<0 -30 -40>, angularTarget=<0 0 0>}
      localJointOffset=<1.5 0 0>
     }]
    planarLimits [{drivingBone {boneName=dyn_tie_MSTR}, planeTransform {location=<-40 1.5 0>, rotation=<rot 90 90 0>}}]
    simSpaceSettings {maxAngularAcceleration=10, maxAngularVelocity=10, simSpaceAngularAlpha=0.5}
   }
   n5 = Anim Dynamics - Bone: dyn_tie_02(ComponentPose=@n4, bAlphaBoolEnabled=True, ExternalForce="0,0,0") {
    alphaInputType=Bool
    angularDampingOverride=0.9998
    bOverrideAngularDamping=true
    bUseGravityOverride=true
    simulationSpace=Actor
    boundBone {boneName=dyn_tie_02}
    componentAppliedLinearAccClamp=<1000 1000 1000>
    componentLinearAccScale=<0.7 0.7 1>
    componentLinearVelScale=<0.8 0.8 1>
    gravityOverride {z=-980}
    physicsBodyDefinitions [{
      collisionType=CustomSphere
      sphereCollisionRadius=1.5
      boundBone {boneName=dyn_tie_02}
      boxExtents=<12 12 12>
      constraintSetup {angularLimitsMax=<5 30 30>, angularLimitsMin=<-5 -30 -40>, angularTarget=<0 0 0>}
      localJointOffset=<1.5 0 0>
     }]
    planarLimits [{drivingBone {boneName=dyn_tie_MSTR}, planeTransform {location=<-40 1.5 0>, rotation=<rot 90 90 0>}}]
    simSpaceSettings {maxAngularAcceleration=10, maxAngularVelocity=10, simSpaceAngularAlpha=0.5}
   }
   n6 = Anim Dynamics - Bone: dyn_tie_03(ComponentPose=@n5, bAlphaBoolEnabled=True, ExternalForce="0,0,0") {
    alphaInputType=Bool
    angularDampingOverride=0.999
    bOverrideAngularDamping=true
    bUseGravityOverride=true
    simulationSpace=Actor
    boundBone {boneName=dyn_tie_03}
    componentAppliedLinearAccClamp=<1000 1000 1000>
    componentLinearAccScale=<0.7 0.7 1>
    componentLinearVelScale=<0.8 0.8 1>
    gravityOverride {z=-980}
    physicsBodyDefinitions [{
      collisionType=CustomSphere
      sphereCollisionRadius=1.5
      boundBone {boneName=dyn_tie_03}
      boxExtents=<11 11 11>
      constraintSetup {angularLimitsMax=<5 30 30>, angularLimitsMin=<-5 -30 -40>, angularTarget=<0 0 0>}
      localJointOffset=<1.5 0 0>
     }]
    planarLimits [{drivingBone {boneName=dyn_tie_MSTR}, planeTransform {location=<-40 1.5 0>, rotation=<rot 90 90 0>}}]
    simSpaceSettings {maxAngularAcceleration=10, maxAngularVelocity=10, simSpaceAngularAlpha=0.5}
   }
   Anim Dynamics - Bone: dyn_tie_04(ComponentPose=@n6, bAlphaBoolEnabled=True, ExternalForce="0,0,0") {
    alphaInputType=Bool
    angularDampingOverride=0.999
    bOverrideAngularDamping=true
    bUseGravityOverride=true
    simulationSpace=Actor
    boundBone {boneName=dyn_tie_04}
    componentAppliedLinearAccClamp=<1000 1000 1000>
    componentLinearAccScale=<0.7 0.7 1>
    componentLinearVelScale=<0.8 0.8 1>
    gravityOverride {z=-980}
    physicsBodyDefinitions [{
      collisionType=CustomSphere
      sphereCollisionRadius=1.5
      boundBone {boneName=dyn_tie_04}
      boxExtents=<10 10 10>
      constraintSetup {angularLimitsMax=<5 30 30>, angularLimitsMin=<-5 -30 -40>, angularTarget=<0 0 0>}
      localJointOffset=<1.5 0 0>
     }]
    planarLimits [{drivingBone {boneName=dyn_tie_MSTR}, planeTransform {location=<-40 1.5 0>, rotation=<rot 90 90 0>}}]
    simSpaceSettings {maxAngularAcceleration=10, maxAngularVelocity=10, simSpaceAngularAlpha=0.5}
   }
   n7 = Anim Dynamics - Bone: dyn_tie_01(ComponentPose=Local To Component(LocalPose=Use cached pose 'FrontEndPose'()), bAlphaBoolEnabled=True, ExternalForce="0,0,0") {
    alphaInputType=Bool
    angularDampingOverride=0.9999
    bOverrideAngularDamping=true
    bUseGravityOverride=true
    simulationSpace=Actor
    boundBone {boneName=dyn_tie_01}
    componentAppliedLinearAccClamp=<1000 1000 1000>
    componentLinearAccScale=<0.2 0.2 0.5>
    componentLinearVelScale=<0.5 0.5 0.5>
    gravityOverride {z=-980}
    physicsBodyDefinitions [{
      collisionType=CustomSphere
      sphereCollisionRadius=1.5
      boundBone {boneName=dyn_tie_01}
      boxExtents=<10 10 10>
      constraintSetup {angularLimitsMax=<0 20 0>, angularLimitsMin=<0 -40 -20>, angularTarget=<0 0 0>}
      localJointOffset=<1.2 0 0>
     }]
    planarLimits [{drivingBone {boneName=dyn_tie_MSTR}, planeTransform {location=<-40 1.5 0>, rotation=<rot 90 90 0>}}]
    simSpaceSettings {maxAngularAcceleration=10, maxAngularVelocity=5, simSpaceAngularAlpha=1}
   }
   n8 = Anim Dynamics - Bone: dyn_tie_02(ComponentPose=@n7, bAlphaBoolEnabled=True, ExternalForce="0,0,0") {
    alphaInputType=Bool
    angularDampingOverride=0.9999
    bOverrideAngularDamping=true
    bUseGravityOverride=true
    simulationSpace=Actor
    boundBone {boneName=dyn_tie_02}
    componentAppliedLinearAccClamp=<1000 1000 1000>
    componentLinearAccScale=<0.2 0.2 0.5>
    componentLinearVelScale=<0.5 0.5 0.5>
    gravityOverride {z=-980}
    physicsBodyDefinitions [{
      collisionType=CustomSphere
      sphereCollisionRadius=1.5
      boundBone {boneName=dyn_tie_02}
      boxExtents=<10 10 10>
      constraintSetup {angularLimitsMax=<2.5 20 20>, angularLimitsMin=<-2.5 -20 -20>, angularTarget=<0 0 0>}
      localJointOffset=<1.2 0 0>
     }]
    planarLimits [{drivingBone {boneName=dyn_tie_MSTR}, planeTransform {location=<-40 1.5 0>, rotation=<rot 90 90 0>}}]
    simSpaceSettings {maxAngularAcceleration=10, maxAngularVelocity=5, simSpaceAngularAlpha=1}
   }
   n9 = Anim Dynamics - Bone: dyn_tie_03(ComponentPose=@n8, bAlphaBoolEnabled=True, ExternalForce="0,0,0") {
    alphaInputType=Bool
    angularDampingOverride=1
    bOverrideAngularDamping=true
    bUseGravityOverride=true
    simulationSpace=Actor
    boundBone {boneName=dyn_tie_03}
    componentAppliedLinearAccClamp=<1000 1000 1000>
    componentLinearAccScale=<0.2 0.2 0.5>
    componentLinearVelScale=<0.5 0.5 0.5>
    gravityOverride {z=-980}
    physicsBodyDefinitions [{
      collisionType=CustomSphere
      sphereCollisionRadius=1.5
      boundBone {boneName=dyn_tie_03}
      boxExtents=<10 10 10>
      constraintSetup {angularLimitsMax=<2.5 20 20>, angularLimitsMin=<-2.5 -20 -20>, angularTarget=<0 0 0>}
      localJointOffset=<1.2 0 0>
     }]
    planarLimits [{drivingBone {boneName=dyn_tie_MSTR}, planeTransform {location=<-40 1.5 0>, rotation=<rot 90 90 0>}}]
    simSpaceSettings {maxAngularAcceleration=10, maxAngularVelocity=5, simSpaceAngularAlpha=1}
   }
   n10 = Anim Dynamics - Bone: dyn_tie_01(ComponentPose=@n9, bAlphaBoolEnabled=True, ExternalForce="0,0,0") {
    alphaInputType=Bool
    angularDampingOverride=0.9998
    bOverrideAngularDamping=true
    bUseGravityOverride=true
    simulationSpace=Actor
    boundBone {boneName=dyn_tie_01}
    componentAppliedLinearAccClamp=<1000 1000 1000>
    componentLinearAccScale=<0.7 0.7 1>
    componentLinearVelScale=<0.8 0.8 1>
    gravityOverride {z=-980}
    physicsBodyDefinitions [{
      collisionType=CustomSphere
      sphereCollisionRadius=1.5
      boundBone {boneName=dyn_tie_01}
      boxExtents=<13 13 13>
      constraintSetup {angularLimitsMax=<0 30 0>, angularLimitsMin=<0 -40 -40>, angularTarget=<0 0 0>}
      localJointOffset=<1.5 0 0>
     }]
    planarLimits [{drivingBone {boneName=dyn_tie_MSTR}, planeTransform {location=<-40 1.5 0>, rotation=<rot 90 90 0>}}]
    simSpaceSettings {maxAngularAcceleration=10, maxAngularVelocity=10, simSpaceAngularAlpha=0.1}
   }
   n11 = Anim Dynamics - Bone: dyn_tie_02(ComponentPose=@n10, bAlphaBoolEnabled=True, ExternalForce="0,0,0") {
    alphaInputType=Bool
    angularDampingOverride=0.9998
    bOverrideAngularDamping=true
    bUseGravityOverride=true
    simulationSpace=Actor
    boundBone {boneName=dyn_tie_02}
    componentAppliedLinearAccClamp=<1000 1000 1000>
    componentLinearAccScale=<0.7 0.7 1>
    componentLinearVelScale=<0.8 0.8 1>
    gravityOverride {z=-980}
    physicsBodyDefinitions [{
      collisionType=CustomSphere
      sphereCollisionRadius=1.5
      boundBone {boneName=dyn_tie_02}
      boxExtents=<12 12 12>
      constraintSetup {angularLimitsMax=<5 30 30>, angularLimitsMin=<-5 -40 -40>, angularTarget=<0 0 0>}
      localJointOffset=<1.5 0 0>
     }]
    planarLimits [{drivingBone {boneName=dyn_tie_MSTR}, planeTransform {location=<-40 1.5 0>, rotation=<rot 90 90 0>}}]
    simSpaceSettings {maxAngularAcceleration=10, maxAngularVelocity=10, simSpaceAngularAlpha=0.1}
   }
   n12 = Anim Dynamics - Bone: dyn_tie_03(ComponentPose=@n11, bAlphaBoolEnabled=True, ExternalForce="0,0,0") {
    alphaInputType=Bool
    angularDampingOverride=0.999
    bOverrideAngularDamping=true
    bUseGravityOverride=true
    simulationSpace=Actor
    boundBone {boneName=dyn_tie_03}
    componentAppliedLinearAccClamp=<1000 1000 1000>
    componentLinearAccScale=<0.7 0.7 1>
    componentLinearVelScale=<0.8 0.8 1>
    gravityOverride {z=-980}
    physicsBodyDefinitions [{
      collisionType=CustomSphere
      sphereCollisionRadius=1.5
      boundBone {boneName=dyn_tie_03}
      boxExtents=<11 11 11>
      constraintSetup {angularLimitsMax=<5 30 30>, angularLimitsMin=<-5 -30 -40>, angularTarget=<0 0 0>}
      localJointOffset=<1.5 0 0>
     }]
    planarLimits [{drivingBone {boneName=dyn_tie_MSTR}, planeTransform {location=<-40 1.5 0>, rotation=<rot 90 90 0>}}]
    simSpaceSettings {maxAngularAcceleration=10, maxAngularVelocity=10, simSpaceAngularAlpha=0.1}
   }
   n13 = Anim Dynamics - Bone: dyn_tie_04(ComponentPose=@n12, bAlphaBoolEnabled=True, ExternalForce="0,0,0") {
    alphaInputType=Bool
    angularDampingOverride=0.999
    bOverrideAngularDamping=true
    bUseGravityOverride=true
    simulationSpace=Actor
    boundBone {boneName=dyn_tie_04}
    componentAppliedLinearAccClamp=<1000 1000 1000>
    componentLinearAccScale=<0.7 0.7 1>
    componentLinearVelScale=<0.8 0.8 1>
    gravityOverride {z=-980}
    physicsBodyDefinitions [{
      collisionType=CustomSphere
      sphereCollisionRadius=1.5
      boundBone {boneName=dyn_tie_04}
      boxExtents=<10 10 10>
      constraintSetup {angularLimitsMax=<5 30 30>, angularLimitsMin=<-5 -30 -40>, angularTarget=<0 0 0>}
      localJointOffset=<1.5 0 0>
     }]
    planarLimits [{drivingBone {boneName=dyn_tie_MSTR}, planeTransform {location=<-40 1.5 0>, rotation=<rot 90 90 0>}}]
    simSpaceSettings {maxAngularAcceleration=10, maxAngularVelocity=10, simSpaceAngularAlpha=0.1}
   }
   n14 = Component To Local(ComponentPose=@n13)
   n15 = Anim Dynamics - Bone: dyn_tie_01(ComponentPose=Local To Component(LocalPose=Use cached pose 'FrontEndPose'()), bAlphaBoolEnabled=True, ExternalForce="0,0,0") {
    alphaInputType=Bool
    angularDampingOverride=0.9999
    bOverrideAngularDamping=true
    bUseGravityOverride=true
    simulationSpace=Actor
    boundBone {boneName=dyn_tie_01}
    componentAppliedLinearAccClamp=<1000 1000 1000>
    componentLinearAccScale=<0.7 0.7 1>
    componentLinearVelScale=<0.8 0.8 1>
    gravityOverride {z=-980}
    physicsBodyDefinitions [{
      collisionType=CustomSphere
      sphereCollisionRadius=1.5
      boundBone {boneName=dyn_tie_01}
      boxExtents=<13 13 13>
      constraintSetup {angularLimitsMax=<0 30 0>, angularLimitsMin=<0 -30 -40>, angularTarget=<0 0 0>}
      localJointOffset=<1.5 0 0>
     }]
    planarLimits [{drivingBone {boneName=dyn_tie_MSTR}, planeTransform {location=<-40 1.5 0>, rotation=<rot 90 90 0>}}]
    simSpaceSettings {maxAngularAcceleration=10, maxAngularVelocity=5, simSpaceAngularAlpha=0.5}
   }
   n16 = Anim Dynamics - Bone: dyn_tie_03(ComponentPose=@n15, bAlphaBoolEnabled=True, ExternalForce="0,0,0") {
    alphaInputType=Bool
    angularDampingOverride=1
    bOverrideAngularDamping=true
    bUseGravityOverride=true
    simulationSpace=Actor
    boundBone {boneName=dyn_tie_03}
    componentAppliedLinearAccClamp=<1000 1000 1000>
    componentLinearAccScale=<0.7 0.7 1>
    componentLinearVelScale=<0.8 0.8 1>
    gravityOverride {z=-980}
    physicsBodyDefinitions [{
      collisionType=CustomSphere
      sphereCollisionRadius=1.5
      boundBone {boneName=dyn_tie_03}
      boxExtents=<11 11 11>
      constraintSetup {angularLimitsMax=<2.5 30 30>, angularLimitsMin=<-2.5 -30 -40>, angularTarget=<0 0 0>}
      localJointOffset=<1.5 0 0>
     }]
    planarLimits [{drivingBone {boneName=dyn_tie_MSTR}, planeTransform {location=<-40 1.5 0>, rotation=<rot 90 90 0>}}]
    simSpaceSettings {maxAngularAcceleration=10, maxAngularVelocity=5, simSpaceAngularAlpha=0.5}
   }
   n17 = Anim Dynamics - Bone: dyn_tie_01(ComponentPose=@n16, bAlphaBoolEnabled=True, ExternalForce="0,0,0") {
    alphaInputType=Bool
    angularDampingOverride=0.9998
    bOverrideAngularDamping=true
    bUseGravityOverride=true
    simulationSpace=Actor
    boundBone {boneName=dyn_tie_01}
    componentAppliedLinearAccClamp=<1000 1000 1000>
    componentLinearAccScale=<0.7 0.7 1>
    componentLinearVelScale=<0.8 0.8 1>
    gravityOverride {z=-980}
    physicsBodyDefinitions [{
      collisionType=CustomSphere
      sphereCollisionRadius=1.5
      boundBone {boneName=dyn_tie_01}
      boxExtents=<13 13 13>
      constraintSetup {angularLimitsMax=<0 30 0>, angularLimitsMin=<0 -30 -40>, angularTarget=<0 0 0>}
      localJointOffset=<1.5 0 0>
     }]
    planarLimits [{drivingBone {boneName=dyn_tie_MSTR}, planeTransform {location=<-40 1.5 0>, rotation=<rot 90 90 0>}}]
    simSpaceSettings {maxAngularAcceleration=10, maxAngularVelocity=10, simSpaceAngularAlpha=0.5}
   }
   n18 = Anim Dynamics - Bone: dyn_tie_02(ComponentPose=@n17, bAlphaBoolEnabled=True, ExternalForce="0,0,0") {
    alphaInputType=Bool
    angularDampingOverride=0.9998
    bOverrideAngularDamping=true
    bUseGravityOverride=true
    simulationSpace=Actor
    boundBone {boneName=dyn_tie_02}
    componentAppliedLinearAccClamp=<1000 1000 1000>
    componentLinearAccScale=<0.7 0.7 1>
    componentLinearVelScale=<0.8 0.8 1>
    gravityOverride {z=-980}
    physicsBodyDefinitions [{
      collisionType=CustomSphere
      sphereCollisionRadius=1.5
      boundBone {boneName=dyn_tie_02}
      boxExtents=<12 12 12>
      constraintSetup {angularLimitsMax=<5 30 30>, angularLimitsMin=<-5 -30 -40>, angularTarget=<0 0 0>}
      localJointOffset=<1.5 0 0>
     }]
    planarLimits [{drivingBone {boneName=dyn_tie_MSTR}, planeTransform {location=<-40 1.5 0>, rotation=<rot 90 90 0>}}]
    simSpaceSettings {maxAngularAcceleration=10, maxAngularVelocity=10, simSpaceAngularAlpha=0.5}
   }
   n19 = Anim Dynamics - Bone: dyn_tie_03(ComponentPose=@n18, bAlphaBoolEnabled=True, ExternalForce="0,0,0") {
    alphaInputType=Bool
    angularDampingOverride=0.999
    bOverrideAngularDamping=true
    bUseGravityOverride=true
    simulationSpace=Actor
    boundBone {boneName=dyn_tie_03}
    componentAppliedLinearAccClamp=<1000 1000 1000>
    componentLinearAccScale=<0.7 0.7 1>
    componentLinearVelScale=<0.8 0.8 1>
    gravityOverride {z=-980}
    physicsBodyDefinitions [{
      collisionType=CustomSphere
      sphereCollisionRadius=1.5
      boundBone {boneName=dyn_tie_03}
      boxExtents=<11 11 11>
      constraintSetup {angularLimitsMax=<5 30 30>, angularLimitsMin=<-5 -30 -40>, angularTarget=<0 0 0>}
      localJointOffset=<1.5 0 0>
     }]
    planarLimits [{drivingBone {boneName=dyn_tie_MSTR}, planeTransform {location=<-40 1.5 0>, rotation=<rot 90 90 0>}}]
    simSpaceSettings {maxAngularAcceleration=10, maxAngularVelocity=10, simSpaceAngularAlpha=0.5}
   }
   n20 = Anim Dynamics - Bone: dyn_tie_04(ComponentPose=@n19, bAlphaBoolEnabled=True, ExternalForce="0,0,0") {
    alphaInputType=Bool
    angularDampingOverride=0.999
    bOverrideAngularDamping=true
    bUseGravityOverride=true
    simulationSpace=Actor
    boundBone {boneName=dyn_tie_04}
    componentAppliedLinearAccClamp=<1000 1000 1000>
    componentLinearAccScale=<0.7 0.7 1>
    componentLinearVelScale=<0.8 0.8 1>
    gravityOverride {z=-980}
    physicsBodyDefinitions [{
      collisionType=CustomSphere
      sphereCollisionRadius=1.5
      boundBone {boneName=dyn_tie_04}
      boxExtents=<10 10 10>
      constraintSetup {angularLimitsMax=<5 30 30>, angularLimitsMin=<-5 -30 -40>, angularTarget=<0 0 0>}
      localJointOffset=<1.5 0 0>
     }]
    planarLimits [{drivingBone {boneName=dyn_tie_MSTR}, planeTransform {location=<-40 1.5 0>, rotation=<rot 90 90 0>}}]
    simSpaceSettings {maxAngularAcceleration=10, maxAngularVelocity=10, simSpaceAngularAlpha=0.5}
   }
   n21 = Component To Local(ComponentPose=@n20)
   n22 = Get bIsOnPawn() -> (bIsOnPawn:bool)
   n23 = Anim Dynamics - Bone: dyn_bag(
    ComponentPose=Local To Component(LocalPose=Blend Poses by bool(BlendPose_0=@n14, BlendPose_1=@n21, bActiveValue=@n22))
    Alpha=1
    ExternalForce="0,0,0"
   ) {
    angularDampingOverride=0.99
    bOverrideAngularDamping=true
    bUseGravityOverride=true
    simulationSpace=Actor
    boundBone {boneName=dyn_bag}
    componentAppliedLinearAccClamp=<800 800 800>
    componentLinearAccScale=<1 1 1>
    componentLinearVelScale=<1 1 1>
    gravityOverride {z=-980}
    physicsBodyDefinitions [{
      boundBone {boneName=dyn_bag}
      boxExtents=<10 10 10>
      constraintSetup {angularLimitsMax=<15 15 15>, angularLimitsMin=<-15 -15 -15>, angularTarget=<0 0 0>}
      localJointOffset=<1 0 0>
     }]
   }
   Output Pose(Result=Component To Local(ComponentPose=@n23)) {layerGroup=None}
   AnimGraphNode_SubInput_0 = Input Pose()
   n24 = "Transform (Modify) Bone - Bone: FruitFire_pelvis"(ComponentPose=Local To Component(LocalPose=@AnimGraphNode_SubInput_0), Alpha=1, Translation="0,0,0", Rotation="0,0,0", Scale="1,1,1") {
    rotationMode=BMM_Replace
    rotationSpace=BCS_WorldSpace
    translationMode=BMM_Replace
    translationSpace=BCS_WorldSpace
    boneToModify {boneName=FruitFire_pelvis}
   }
   n25 = "Transform (Modify) Bone - Bone: FruitFire_thigh_l"(ComponentPose=@n24, Alpha=1, Translation="0,0,0", Rotation="0,0,0", Scale="1,1,1") {
    rotationMode=BMM_Replace
    rotationSpace=BCS_WorldSpace
    translationMode=BMM_Replace
    translationSpace=BCS_WorldSpace
    boneToModify {boneName=FruitFire_thigh_l}
   }
   n26 = "Transform (Modify) Bone - Bone: FruitFire_thigh_r"(ComponentPose=@n25, Alpha=1, Translation="0,0,0", Rotation="0,0,0", Scale="1,1,1") {
    rotationMode=BMM_Replace
    rotationSpace=BCS_WorldSpace
    translationMode=BMM_Replace
    translationSpace=BCS_WorldSpace
    boneToModify {boneName=FruitFire_thigh_r}
   }
   n27 = "Transform (Modify) Bone - Bone: FruitFire_calf_l"(ComponentPose=@n26, Alpha=1, Translation="0,0,0", Rotation="0,0,0", Scale="1,1,1") {
    rotationMode=BMM_Replace
    rotationSpace=BCS_WorldSpace
    translationMode=BMM_Replace
    translationSpace=BCS_WorldSpace
    boneToModify {boneName=FruitFire_calf_l}
   }
   n28 = "Transform (Modify) Bone - Bone: FruitFire_calf_r"(ComponentPose=@n27, Alpha=1, Translation="0,0,0", Rotation="0,0,0", Scale="1,1,1") {
    rotationMode=BMM_Replace
    rotationSpace=BCS_WorldSpace
    translationMode=BMM_Replace
    translationSpace=BCS_WorldSpace
    boneToModify {boneName=FruitFire_calf_r}
   }
   Save cached pose 'PreSimPose'(Pose=Component To Local(ComponentPose=@n28)) {cacheName=PreSimPose}
   n29 = RigidBody(ComponentPose=Local To Component(LocalPose=Use cached pose 'PreSimPose'())) {
    overridePhysicsAsset=&/FruitFire_Nudge_Axe_Cord_Physics
    simulationSpace=BaseBoneSpace
    baseBoneRef {boneName=Melee_TwoHanded_Handle}
    componentAppliedLinearAccClamp=<2000 2000 2000>
    simSpaceSettings {
     dampingAlpha=0
     maxAngularAcceleration=15
     maxAngularVelocity=2
     maxLinearAcceleration=1000
     maxLinearVelocity=200
     worldAlpha=1
     externalLinearDragV=<0.5 0.5>
    }
   }
   Save cached pose 'PhysOrDynPose'(Pose=Blend Poses by bool(BlendPose_0=Component To Local(ComponentPose=@n29), BlendPose_1=Use cached pose 'PawnPose'(), bActiveValue=Get bRigidBodyEnabled())) {cacheName=PhysOrDynPose}
   n30 = Anim Dynamics - Bone: dyn_cord_01(ComponentPose=Local To Component(LocalPose=Input Pose() {name=InPose_3}), Alpha=1, ExternalForce="0,0,0") {
    angularDampingOverride=0.9999
    angularSpringConstant=50
    bAngularSpring=true
    bOverrideAngularDamping=true
    bUseGravityOverride=true
    simulationSpace=Actor
    boundBone {boneName=dyn_cord_01}
    componentAppliedLinearAccClamp=<1000 1000 1000>
    componentLinearAccScale=<0.8 0.8 0.8>
    componentLinearVelScale=<1 1 1>
    gravityOverride {z=-900}
    physicsBodyDefinitions [{
      boundBone {boneName=dyn_cord_01}
      boxExtents=<10 10 10>
      constraintSetup {angularLimitsMax=<1 0 35>, angularLimitsMin=<-1 0 -10>, angularTarget=<1 0 0>}
      localJointOffset=<1 0 0>
     }]
    simSpaceSettings {maxAngularAcceleration=8, maxAngularVelocity=8}
   }
   n31 = Anim Dynamics - Bone: dyn_cord_02(ComponentPose=@n30, Alpha=1, ExternalForce="0,0,0") {
    angularDampingOverride=0.9999
    angularSpringConstant=50
    bAngularSpring=true
    bOverrideAngularDamping=true
    bUseGravityOverride=true
    simulationSpace=Actor
    boundBone {boneName=dyn_cord_02}
    componentAppliedLinearAccClamp=<1000 1000 1000>
    componentLinearAccScale=<0.8 0.8 0.8>
    componentLinearVelScale=<1 1 1>
    gravityOverride {z=-900}
    physicsBodyDefinitions [{
      boundBone {boneName=dyn_cord_02}
      boxExtents=<10 10 10>
      constraintSetup {angularLimitsMax=<1 35 35>, angularLimitsMin=<-1 0 -30>, angularTarget=<1 0 0>}
      localJointOffset=<1 0 0>
     }]
    simSpaceSettings {maxAngularAcceleration=8, maxAngularVelocity=8}
   }
   n32 = Anim Dynamics - Bone: dyn_cord_03(ComponentPose=@n31, Alpha=1, ExternalForce="0,0,0") {
    angularDampingOverride=0.9999
    angularSpringConstant=50
    bAngularSpring=true
    bOverrideAngularDamping=true
    bUseGravityOverride=true
    simulationSpace=Actor
    boundBone {boneName=dyn_cord_03}
    componentAppliedLinearAccClamp=<1000 1000 1000>
    componentLinearAccScale=<0.8 0.8 0.8>
    componentLinearVelScale=<1 1 1>
    gravityOverride {z=-900}
    physicsBodyDefinitions [{
      boundBone {boneName=dyn_cord_03}
      boxExtents=<10 10 10>
      constraintSetup {angularLimitsMax=<1 35 35>, angularLimitsMin=<-1 -35 -30>, angularTarget=<1 0 0>}
      localJointOffset=<1 0 0>
     }]
    simSpaceSettings {maxAngularAcceleration=8, maxAngularVelocity=8}
   }
   n33 = Anim Dynamics - Bone: dyn_cord_04(ComponentPose=@n32, Alpha=1, ExternalForce="0,0,0") {
    angularDampingOverride=0.9999
    angularSpringConstant=50
    bAngularSpring=true
    bOverrideAngularDamping=true
    bUseGravityOverride=true
    simulationSpace=Actor
    boundBone {boneName=dyn_cord_04}
    componentAppliedLinearAccClamp=<1000 1000 1000>
    componentLinearAccScale=<0.8 0.8 0.8>
    componentLinearVelScale=<1 1 1>
    gravityOverride {z=-900}
    physicsBodyDefinitions [{
      boundBone {boneName=dyn_cord_04}
      boxExtents=<10 10 10>
      constraintSetup {angularLimitsMax=<1 35 35>, angularLimitsMin=<-1 -35 -10>, angularTarget=<1 0 0>}
      localJointOffset=<1 0 0>
     }]
    simSpaceSettings {maxAngularAcceleration=8, maxAngularVelocity=8}
   }
   n34 = Anim Dynamics - Bone: dyn_cord_05(ComponentPose=@n33, Alpha=1, ExternalForce="0,0,0") {
    angularDampingOverride=0.9999
    angularSpringConstant=50
    bAngularSpring=true
    bOverrideAngularDamping=true
    bUseGravityOverride=true
    simulationSpace=Actor
    boundBone {boneName=dyn_cord_05}
    componentAppliedLinearAccClamp=<1000 1000 1000>
    componentLinearAccScale=<0.8 0.8 0.8>
    componentLinearVelScale=<1 1 1>
    gravityOverride {z=-900}
    physicsBodyDefinitions [{
      boundBone {boneName=dyn_cord_05}
      boxExtents=<10 10 10>
      constraintSetup {angularLimitsMax=<1 35 35>, angularLimitsMin=<-1 -35 -10>, angularTarget=<1 0 0>}
      localJointOffset=<1 0 0>
     }]
    simSpaceSettings {maxAngularAcceleration=8, maxAngularVelocity=8}
   }
   n35 = Anim Dynamics - Bone: dyn_cord_06(ComponentPose=@n34, Alpha=1, ExternalForce="0,0,0") {
    angularDampingOverride=0.9999
    angularSpringConstant=50
    bAngularSpring=true
    bOverrideAngularDamping=true
    bUseGravityOverride=true
    simulationSpace=Actor
    boundBone {boneName=dyn_cord_06}
    componentAppliedLinearAccClamp=<1000 1000 1000>
    componentLinearAccScale=<0.8 0.8 0.8>
    componentLinearVelScale=<1 1 1>
    gravityOverride {z=-980}
    physicsBodyDefinitions [{
      boundBone {boneName=dyn_cord_06}
      boxExtents=<10 12 12>
      constraintSetup {angularLimitsMax=<1 35 35>, angularLimitsMin=<-1 -35 -10>, angularTarget=<1 0 0>}
      localJointOffset=<1 0 0>
     }]
    simSpaceSettings {maxAngularAcceleration=8, maxAngularVelocity=8}
   }
   n36 = Anim Dynamics - Bone: dyn_cord_01(ComponentPose=Local To Component(LocalPose=Input Pose() {name=InPose_4}), Alpha=1, ExternalForce="0,0,0") {
    angularDampingOverride=0.999
    angularSpringConstant=50
    bAngularSpring=true
    bOverrideAngularDamping=true
    bUseGravityOverride=true
    simulationSpace=Actor
    boundBone {boneName=dyn_cord_01}
    componentAppliedLinearAccClamp=<1000 1000 1000>
    componentLinearAccScale=<0.8 0.8 0.8>
    componentLinearVelScale=<1 1 1>
    gravityOverride {z=-900}
    physicsBodyDefinitions [{
      boundBone {boneName=dyn_cord_01}
      boxExtents=<10 10 10>
      constraintSetup {angularLimitsMax=<1 35 35>, angularLimitsMin=<-1 0 -10>, angularTarget=<1 0 0>}
      localJointOffset=<1 0 0>
     }]
    simSpaceSettings {maxAngularAcceleration=8, maxAngularVelocity=8, simSpaceAngularAlpha=1}
   }
   n37 = Anim Dynamics - Bone: dyn_cord_02(ComponentPose=@n36, Alpha=1, ExternalForce="0,0,0") {
    angularDampingOverride=0.999
    angularSpringConstant=50
    bAngularSpring=true
    bOverrideAngularDamping=true
    bUseGravityOverride=true
    simulationSpace=Actor
    boundBone {boneName=dyn_cord_02}
    componentAppliedLinearAccClamp=<1000 1000 1000>
    componentLinearAccScale=<0.8 0.8 0.8>
    componentLinearVelScale=<1 1 1>
    gravityOverride {z=-900}
    physicsBodyDefinitions [{
      boundBone {boneName=dyn_cord_02}
      boxExtents=<10 10 10>
      constraintSetup {angularLimitsMax=<1 35 35>, angularLimitsMin=<-1 0 -10>, angularTarget=<1 0 0>}
      localJointOffset=<1 0 0>
     }]
    simSpaceSettings {maxAngularAcceleration=8, maxAngularVelocity=8, simSpaceAngularAlpha=1}
   }
   n38 = Anim Dynamics - Bone: dyn_cord_03(ComponentPose=@n37, Alpha=1, ExternalForce="0,0,0") {
    angularDampingOverride=0.999
    angularSpringConstant=50
    bAngularSpring=true
    bOverrideAngularDamping=true
    bUseGravityOverride=true
    simulationSpace=Actor
    boundBone {boneName=dyn_cord_03}
    componentAppliedLinearAccClamp=<1000 1000 1000>
    componentLinearAccScale=<0.8 0.8 0.8>
    componentLinearVelScale=<1 1 1>
    gravityOverride {z=-900}
    physicsBodyDefinitions [{
      boundBone {boneName=dyn_cord_03}
      boxExtents=<10 10 10>
      constraintSetup {angularLimitsMax=<1 35 35>, angularLimitsMin=<-1 -35 -10>, angularTarget=<1 0 0>}
      localJointOffset=<1 0 0>
     }]
    simSpaceSettings {maxAngularAcceleration=8, maxAngularVelocity=8, simSpaceAngularAlpha=1}
   }
   n39 = Anim Dynamics - Bone: dyn_cord_04(ComponentPose=@n38, Alpha=1, ExternalForce="0,0,0") {
    angularDampingOverride=0.999
    angularSpringConstant=50
    bAngularSpring=true
    bOverrideAngularDamping=true
    bUseGravityOverride=true
    simulationSpace=Actor
    boundBone {boneName=dyn_cord_04}
    componentAppliedLinearAccClamp=<1000 1000 1000>
    componentLinearAccScale=<0.8 0.8 0.8>
    componentLinearVelScale=<1 1 1>
    gravityOverride {z=-900}
    physicsBodyDefinitions [{
      boundBone {boneName=dyn_cord_04}
      boxExtents=<10 10 10>
      constraintSetup {angularLimitsMax=<1 35 35>, angularLimitsMin=<-1 -35 -10>, angularTarget=<1 0 0>}
      localJointOffset=<1 0 0>
     }]
    simSpaceSettings {maxAngularAcceleration=8, maxAngularVelocity=8, simSpaceAngularAlpha=1}
   }
   n40 = Anim Dynamics - Bone: dyn_cord_05(ComponentPose=@n39, Alpha=1, ExternalForce="0,0,0") {
    angularDampingOverride=0.999
    angularSpringConstant=50
    bAngularSpring=true
    bOverrideAngularDamping=true
    bUseGravityOverride=true
    simulationSpace=Actor
    boundBone {boneName=dyn_cord_05}
    componentAppliedLinearAccClamp=<1000 1000 1000>
    componentLinearAccScale=<0.8 0.8 0.8>
    componentLinearVelScale=<1 1 1>
    gravityOverride {z=-900}
    physicsBodyDefinitions [{
      boundBone {boneName=dyn_cord_05}
      boxExtents=<10 10 10>
      constraintSetup {angularLimitsMax=<1 35 35>, angularLimitsMin=<-1 -35 -10>, angularTarget=<1 0 0>}
      localJointOffset=<1 0 0>
     }]
    simSpaceSettings {maxAngularAcceleration=8, maxAngularVelocity=8, simSpaceAngularAlpha=1}
   }
   n41 = Anim Dynamics - Bone: dyn_cord_06(ComponentPose=@n40, Alpha=1, ExternalForce="0,0,0") {
    angularDampingOverride=0.999
    angularSpringConstant=50
    bAngularSpring=true
    bOverrideAngularDamping=true
    bUseGravityOverride=true
    simulationSpace=Actor
    boundBone {boneName=dyn_cord_06}
    componentAppliedLinearAccClamp=<1000 1000 1000>
    componentLinearAccScale=<0.8 0.8 0.8>
    componentLinearVelScale=<1 1 1>
    gravityOverride {z=-980}
    physicsBodyDefinitions [{
      boundBone {boneName=dyn_cord_06}
      boxExtents=<10 12 12>
      constraintSetup {angularLimitsMax=<1 35 35>, angularLimitsMin=<-1 -35 -10>, angularTarget=<1 0 0>}
      localJointOffset=<1 0 0>
     }]
    simSpaceSettings {maxAngularAcceleration=8, maxAngularVelocity=8, simSpaceAngularAlpha=1}
   }
   Save cached pose 'PawnPose'(Pose=Blend Poses by bool(BlendPose_0=Component To Local(ComponentPose=@n35), BlendPose_1=Component To Local(ComponentPose=@n41), bActiveValue=Get bIsOnPawn())) {cacheName=PawnPose}
   Save cached pose 'FrontEndPose'(Pose=Blend Poses by bool(BlendPose_0=Use cached pose 'PawnPose'(), BlendPose_1=Use cached pose 'PhysOrDynPose'(), bActiveValue=Get bIsInFrontEnd())) {cacheName=FrontEndPose}
  ]}
 EventGraph {nodes [
   Event BlueprintUpdateAnimation()
   n1 = GetPartSkeletalMeshComponent()
   > Is Valid(InputObject=@n1)
   n2 = GetAttachParent!(self:Scene Component Object Reference=@n1) -> (ReturnValue:Scene Component Object Reference)
   > Is Valid(InputObject=@n2)
   > Set BodyMesh(BodyMesh:Scene Component Object Reference=@n2) -> (Output_Get:Scene Component Object Reference)
   > Set Is NOT FrontEnd(Is NOT FrontEnd:bool=NOT Boolean(A=Get bIsInFrontEnd())) -> (Output_Get:bool)
   n3 = Get BodyMesh() -> (BodyMesh:Scene Component Object Reference)
   n4 = BreakTransform(InTransform=GetSocketTransform(self=@n3, InSocketName=pelvis, TransformSpace=RTS_World))
   > Set pelvis_location(pelvis_location:Vector=@n4) -> (Output_Get:Vector)
   > Set pelvis_rotation(pelvis_rotation:Rotator=@n4.Rotation) -> (Output_Get:Rotator)
   n5 = BreakTransform(InTransform=GetSocketTransform(self=@n3, InSocketName=thigh_l, TransformSpace=RTS_World))
   > Set thigh_left_location(thigh_left_location:Vector=@n5) -> (Output_Get:Vector)
   > Set thigh_left_rotation(thigh_left_rotation:Rotator=@n5.Rotation) -> (Output_Get:Rotator)
   n6 = BreakTransform(InTransform=GetSocketTransform(self=@n3, InSocketName=thigh_r, TransformSpace=RTS_World))
   > Set thigh_right_location(thigh_right_location:Vector=@n6) -> (Output_Get:Vector)
   > Set thigh_right_rotation(thigh_right_rotation:Rotator=@n6.Rotation) -> (Output_Get:Rotator)
   n7 = BreakTransform(InTransform=GetSocketTransform(self=@n3, InSocketName=calf_l, TransformSpace=RTS_World))
   > Set calf_left_location(calf_left_location:Vector=@n7) -> (Output_Get:Vector)
   > Set calf_left_rotation(calf_left_rotation:Rotator=@n7.Rotation) -> (Output_Get:Rotator)
   n8 = BreakTransform(InTransform=GetSocketTransform(self=@n3, InSocketName=calf_r, TransformSpace=RTS_World))
   > Set calf_right_location(calf_right_location:Vector=@n8) -> (Output_Get:Vector)
   > Set calf_right_rotation(calf_right_rotation:Rotator=@n8.Rotation) -> (Output_Get:Rotator)
  ]}
]
variables [
 BodyMesh {type=SceneComponent, default=null}
 calf_left_location {type=Vector}
 calf_left_rotation {type=Rotator}
 calf_right_location {type=Vector}
 calf_right_rotation {type=Rotator}
 Is NOT FrontEnd {type=boolean}
 pelvis_location {type=Vector}
 pelvis_rotation {type=Rotator}
 thigh_left_location {type=Vector}
 thigh_left_rotation {type=Rotator}
 thigh_right_location {type=Vector}
 thigh_right_rotation {type=Rotator}
]
